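# Niagara impact effect for the gold Crew pickaxe cosmetic template, with three built-out emitters; the other eight carry no renderers.
# Flash_Flare bursts one short flare with a light and time-of-day alpha scaling.
# Sparks_Toon spawns brief, bright orange sparks on a sphere, split into two visibility-tagged spawn groups.
# Gold_Dust_Collision throws distance-culled gold motes with gravity, drag and collision, and all three apply front-end scale modifiers.
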NiagaraSystem /BRCosmetics/Effects/Niagara/Systems/CosmeticTemplates/Pickaxe_Impacts/Crew/Gold/NS_Pickaxe_Crew_Impact_Gold
dependencies {
 usedDataInterfaces [&NiagaraDataInterfaceColorCurve, &NiagaraDataInterfaceCurve]
 usedDynamicInputs []
 usedModules [
  &SystemState
  &self:SystemSpawnScript.SystemScriptSource.SystemScriptGraph.NiagaraNodeAssignment_0.SetVariables_#1
 ]
}
emitters {
 DarkBG {rendererClasses [], renderers []}
 Dust {rendererClasses [], renderers []}
 Flash_Flare {
  emitterUpdateScript [
   &EmitterState()
   &SpawnBurst_Instantaneous(Spawn Count=1, Use Spawn Probability=false)
  ]
  particleSpawnScript [
   &InitializeParticle(
    Color=<rgba 1.5 1.5 1.5 1>
    Lifetime=0.25
    Sprite Size Mode=::NewEnumerator3
    Uniform Sprite Size=180
    UsePositionOffset=false
    Write Material Random=true
   ) {module=InitializeParticle001}
   &DynamicMaterialParameters(
    Index 0 Param 1=1
    Index 0 Param 2=1
    Module.Param0WriteEnabled=true
    Module.Param1WriteEnabled=true
    Module.Param2WriteEnabled=true
    Module.Param3WriteEnabled=true
   )
   &CameraOffset(Camera Offset Amount=5)
  ]
  particleUpdateScript [
   &ParticleState()
   &ScaleSpriteSize(
    Uniform Curve Scale=&FN_FrontEndFloatModifier(Gameplay=1, Front End=$System.FrontendScale)
    Uniform Curve Sprite Scale=DataInterface() {propertyValues {Curve {keys [
        {arriveTangent=2.9559, interpMode=RCIM_Cubic, leaveTangent=2.9559, tangentMode=RCTM_User, time=0, value=0.6}
        {arriveTangent=0.0224, interpMode=RCIM_Cubic, leaveTangent=0.0224, tangentMode=RCTM_User, time=0.3, value=1}
        <at 1 0>
       ]}}}
   )
   &ScaleColor(
    Scale Alpha=&FNTimeOfDayScalar(Night Scale=0.75, Morning Scale=1, Evening Scale=1)
    Scale RGB=&MakeVectorFromLinearColorRGB(Color=&ColorFromCurve(ColorCurve=DataInterface() {propertyValues {
        AlphaCurve {keys [<at 0 0>,<at 0.1287 1 RCIM_Cubic>,<at 1 0>]}
        BlueCurve {keys [<at 0.004 0>, {arriveTangent=0.077, interpMode=RCIM_Cubic, leaveTangent=0.077, time=0.1965, value=0}, <at 1 0.0767 RCIM_Cubic>]}
        GreenCurve {keys [<at 0.004 0.3994>, {arriveTangent=0.0462, interpMode=RCIM_Cubic, leaveTangent=0.0462, time=0.1965, value=0.8274}, <at 1 0.4454 RCIM_Cubic>]}
        RedCurve {keys [<at 0.004 1>,<at 0.1965 2 RCIM_Cubic>,<at 1 1 RCIM_Cubic>]}
       }}))
    ScaleA=true
    ScaleRGB=true
   )
   &FN_Light_Attributes(
    Color=$Particles.Color
    Create Color Attribute=true
    Create Exponent Attribute=true
    Create Radius Attribute=true
    Exponent=4
    Radius=100
    Scale Color=&FloatFromCurve(FloatCurve=DataInterface() {propertyValues {Curve {keys [<at 0 1>,<at 1 0>]}}}, Scale Curve=4)
   )
   &FN_Sprite_Attributes(
    Color=&ColorFromCurve(ColorCurve=DataInterface() {propertyValues {
       AlphaCurve {keys [<at 0 3 RCIM_Cubic>,<at 1 0 RCIM_Cubic>]}
       BlueCurve {keys [<at 0 0 RCIM_Cubic>,<at 0.9989 0>]}
       GreenCurve {keys [<at 0 0 RCIM_Cubic>,<at 0.9989 0>]}
       RedCurve {keys [<at 0 0 RCIM_Cubic>,<at 0.9989 0>]}
      }})
    Create Color Attribute=true
    Create SpriteSize Attribute=true
    SpriteSize=&Vector2DFromFloat(Value=&FloatFromCurve(
      FloatCurve=DataInterface() {propertyValues {Curve {keys [
          {arriveTangent=7.1084, interpMode=RCIM_Cubic, leaveTangent=7.1084, tangentMode=RCTM_User, time=0, value=0.375}
          {arriveTangent=-2.821, interpMode=RCIM_Cubic, leaveTangent=-2.821, tangentMode=RCTM_User, time=0.5, value=0.625}
          {tangentMode=RCTM_User, time=1, value=0}
         ]}}}
      Scale Curve=&Multiply_Float(A=120, B=&FN_FrontEndFloatModifier(Gameplay=1, Front End=$System.FrontendScale))
     ))
    Write to Module Namespace=true
   )
  ]
  rendererClasses [&NiagaraSpriteRendererProperties, &NiagaraLightRendererProperties]
  renderers [
   {rendererClass=&NiagaraSpriteRendererProperties}
   {rendererClass=&NiagaraLightRendererProperties, rendererIndex=1}
   {rendererClass=&NiagaraSpriteRendererProperties, rendererIndex=2}
  ]
 }
 Gold_Dust_Collision {
  emitterUpdateScript [
   &EmitterState(
    Enable Distance Culling=true
    Enable Visibility Culling=true
    Life Cycle Mode=::NewEnumerator1
    Loop Behavior=::NewEnumerator1
    Loop Duration=2
    MaxDistance=2500
    Module.bApplySpawnCountScaleByDistance=true
    Module.bUseMaxDistance=true
    Module.bUseMinDistance=false
    Scalability Mode=::NewEnumerator1
    Scale Spawn Count=true
    UseLoopDelay=false
   )
   &SpawnBurst_Instantaneous(
    Spawn Count=&RandomRangeInt(Minimum=5, Maximum=8)
    Use Spawn Probability=false
    UseLoopCountLimit=false
   )
  ]
  particleSpawnScript [
   &InitializeParticle(
    Color=<rgba 5 2.625 0.25 1>
    Lifetime Max=0.7
    Lifetime Min=0.5
    Lifetime Mode=::NewEnumerator1
    Sprite Size Mode=::NewEnumerator1
    Uniform Sprite Size Max=6
    Uniform Sprite Size Min=4
    UsePositionOffset=true
    Write Material Random=true
   )
   &ShapeLocation(
    MaskSpawnGroup=false
    Override Seed=false
    Sphere Radius=&Multiply_Float(A=20, B=&FN_FrontEndFloatModifier(Gameplay=1, Front End=$System.FrontendScale))
    Transform Order:ENiagara_TransformOrder=::NewEnumerator1
    UseOwnerScale=true
   ) {module=ShapeLocation001}
   &AddVelocity(
    Module.Use Falloff Distance=false
    Override Seed=false
    Velocity Mode=::NewEnumerator1
    Velocity Speed=&Multiply_Float(
     A=&RandomRangeFloat(Minimum=10, Maximum=300)
     B=&FN_FrontEndFloatModifier(Gameplay=1, Front End=$System.FrontendScale)
    )
   ) {module=AddVelocity001}
  ]
  particleUpdateScript [
   &ParticleState()
   &AccelerationForce(Acceleration=&Multiply_VectorByFloat(Vector=<0 0 -980>, Float=&FN_FrontEndFloatModifier(Gameplay=1, Front End=$System.FrontendScale)))
   &Drag(Use Linear Drag=true, Use Rotational Drag=true)
   &Collision(
    Age Colliding Particles=false
    KillOnCollision=false
    Module.Control Rest State Via Surface Normal=false
    Module.EnableMaxCollisionCount=false
    Module.Randomize Collision Normal Vector=false
   )
   &SolveForcesAndVelocity(Clamp Velocity=false, Limit Acceleration=false)
   &ScaleSpriteSize(
    Uniform Curve Scale=&FN_FrontEndFloatModifier(Gameplay=1, Front End=$System.FrontendScale)
    Uniform Curve Sprite Scale=DataInterface() {propertyValues {Curve {keys [
        {arriveTangent=9.0725, interpMode=RCIM_Cubic, leaveTangent=3.7536, tangentMode=RCTM_Break, time=0, value=0}
        <at 0.8 1 RCIM_Cubic>
        {arriveTangent=-5.3074, interpMode=RCIM_Cubic, leaveTangent=-5.3074, tangentMode=RCTM_User, time=1, value=0}
       ]}}}
   )
  ]
  rendererClasses [&NiagaraSpriteRendererProperties]
  renderers [{rendererClass=&NiagaraSpriteRendererProperties}]
 }
 SmokePuff {rendererClasses [], renderers []}
 Sparks001 {rendererClasses [], renderers []}
 Sparks_Toon {
  emitterUpdateScript [
   &EmitterState(Life Cycle Mode=::NewEnumerator1, Loop Behavior=::NewEnumerator1, Loop Delay=0.005, Loop Duration=0.25, UseLoopDelay=false)
   &SpawnRate(SpawnRate=10, Use Spawn Probability=false)
   &SpawnBurst_Instantaneous(Spawn Count=5, Use Spawn Probability=false, UseLoopCountLimit=false)
  ]
  particleSpawnScript [
   &FN_SetSpawnGroup(Number of Spawn Groups=2)
   &InitializeParticle(
    Color=<rgba 0.05 0.05 0.05 0.5>
    Lifetime Max=0.5
    Lifetime Min=0.3
    Lifetime Mode=::NewEnumerator1
    Sprite Size Mode=::NewEnumerator1
    Uniform Sprite Size Max=15
    Uniform Sprite Size Min=10
    UsePositionOffset=false
    Write Material Random=true
   )
   &SphereLocation(
    MaskSpawnGroup=false
    Override Seed=false
    Sphere Radius=&Multiply_Float(A=25, B=&FN_FrontEndFloatModifier(Gameplay=1, Front End=$System.FrontendScale))
    Surface Expansion Mode=::NewEnumerator1
    Surface Only=true
    UseNonUniformScale=false
    UseOffset=true
    UseSphereOrientation=false
   )
   &CameraOffset(Camera Offset Amount=1)
   /SetVariables(Particles.VisibilityTag=&SpawnGroup()) {isSetParameters=true}
  ]
  particleUpdateScript [
   &ParticleState()
   &SolveForcesAndVelocity(Clamp Velocity=false, Limit Acceleration=false)
   &Color(
    Color=&ColorFromCurve(ColorCurve=DataInterface() {propertyValues {
       AlphaCurve {keys [<at 0.0065 0>,<at 0.3936 0.5 RCIM_Cubic>,<at 1 0 RCIM_Cubic>]}
       BlueCurve {keys [<at 0 0 RCIM_Cubic>,<at 0.7105 0 RCIM_Cubic>]}
       GreenCurve {keys [<at 0 0.75 RCIM_Cubic>,<at 0.7105 0.0369 RCIM_Cubic>]}
       RedCurve {keys [<at 0 3 RCIM_Cubic>,<at 0.7105 0.3613 RCIM_Cubic>]}
      }})
    Scale Alpha=&FloatFromCurve(
     FloatCurve=DataInterface() {propertyValues {Curve {keys [
         {interpMode=RCIM_Cubic, tangentMode=RCTM_User, time=-0.0063, value=0.0184}
         {arriveTangent=-0.8483, interpMode=RCIM_Cubic, leaveTangent=-0.8483, tangentMode=RCTM_User, time=0.1718, value=1.7147}
         {arriveTangent=-5.0978, interpMode=RCIM_Cubic, leaveTangent=-5.0978, tangentMode=RCTM_User, time=0.2818, value=1.2143}
         {arriveTangent=-0.2783, interpMode=RCIM_Cubic, leaveTangent=-0.2783, tangentMode=RCTM_User, time=0.4, value=0.7}
         {arriveTangent=0.3112, interpMode=RCIM_Cubic, leaveTangent=0.3112, tangentMode=RCTM_User, time=0.5407, value=1.3419}
         {arriveTangent=-3.7119, interpMode=RCIM_Cubic, leaveTangent=-3.7119, tangentMode=RCTM_User, time=0.7242, value=0.3749}
         {interpMode=RCIM_Cubic, tangentMode=RCTM_User, time=1.0011, value=0.002}
        ]}}}
     Scale Curve=0.8
    )
    Scale Color=&VectorFromFloat(Value=50)
   )
   &ScaleSpriteSize(Scale Factor=&Vector2DFromFloat(Value=&FloatFromCurve(
      FloatCurve=DataInterface() {propertyValues {Curve {keys [
          {arriveTangent=4.1302, interpMode=RCIM_Cubic, leaveTangent=4.1302, tangentMode=RCTM_User, time=0, value=0}
          {interpMode=RCIM_Cubic, tangentMode=RCTM_User, time=0.5, value=1}
          {tangentMode=RCTM_User, time=1, value=0}
         ]}}}
      Scale Curve=&Multiply_Float(
       A=&MaskFloatBySpawnGroup(Masked Float=1, Passthrough Float=1.5)
       B=&FN_FrontEndFloatModifier(Gameplay=1, Front End=$User.FrontEndScale)
      )
     )))
  ]
  rendererClasses [&NiagaraSpriteRendererProperties]
  renderers [
   {rendererClass=&NiagaraSpriteRendererProperties}
   {rendererClass=&NiagaraSpriteRendererProperties, rendererIndex=1}
  ]
 }
 SpikeGreen {rendererClasses [], renderers []}
 SpikePink {rendererClasses [], renderers []}
 StarSpike {rendererClasses [], renderers []}
 Wave {rendererClasses [], renderers []}
}
summary {
 systemName=NS_Pickaxe_Crew_Impact_Gold
 emitters [
  {emitterName=Sparks_Toon, rendererClasses [&NiagaraSpriteRendererProperties]}
  {emitterName=Flash_Flare, rendererClasses [&NiagaraSpriteRendererProperties, &NiagaraLightRendererProperties]}
  {emitterName=Gold_Dust_Collision, rendererClasses [&NiagaraSpriteRendererProperties]}
  {emitterName=StarSpike, rendererClasses []}
  {emitterName=SpikePink, rendererClasses []}
  {emitterName=SpikeGreen, rendererClasses []}
  {emitterName=DarkBG, rendererClasses []}
  {emitterName=SmokePuff, rendererClasses []}
  {emitterName=Sparks001, rendererClasses []}
  {emitterName=Wave, rendererClasses []}
  {emitterName=Dust, rendererClasses []}
 ]
 userVariables [
  User.bIsFrontEnd {type=NiagaraBool}
  User.bIsFrontEndPreview {type=NiagaraBool}
  User.FrontEndScale {type=NiagaraFloat, defaultValue=0, description=""}
 ]
}
